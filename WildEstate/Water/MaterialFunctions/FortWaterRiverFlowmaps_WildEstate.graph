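# WildEstate river water material function that builds flowmap normals, foam, scattering and stylized wave lines.
# WaterFlowMap animates normal and height from velocity and depth, while velocity, depth and sim erasure shape foam.
# Two offset cell grids rotate and pan a streak texture along per-cell WaterBodyData flow, then get stepped into FlowMapTexture.
# Depth-based sine shore lines and flow masks tint the incoming attributes' BaseColor and Opacity, with transition and mobile scaling.
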MaterialFunction /WildEstate/Water/MaterialFunctions/FortWaterRiverFlowmaps_WildEstate
graph {nodes [
  n1 = WorldPosition()
  n2 = WorldPosition() {worldPositionShaderOffset=WPT_ExcludeAllShaderOffsets}
  n3 = TextureSample(UVs=Divide(A=ComponentMask(None=@n2.XYZ) {g=true, r=true}, B=ScalarParameter() {defaultValue=10125.3916, parameterName=NoiseSizeForTIles})) {samplerSource=SSM_Wrap_WorldGroupSettings, samplerType=SAMPLERTYPE_Grayscale, texture=/WildEstate/Water/Textures/WildEstate_Water_M}
  Add(A=@n1.XY, B=Multiply(A=Subtract(A=@n3.R) {constB=0.5}, B=ScalarParameter() {defaultValue=240.9168, parameterName=DistortUVTIles}))
  n4 = NamedRerouteUsage()
  n5 = Saturate(None=Constant() {r=1})
  n6 = Multiply(A=@n4, B=@n5)
  n7 = ComponentMask(None=@n6) {r=true}
  n8 = Subtract(A=ComponentMask(None=NamedRerouteUsage()) {b=true}, B=ScalarParameter() {defaultValue=150, group=Flowmaps, parameterName=Velocity Foam Bias})
  n9 = ScalarParameter() {defaultValue=99.4355, group=Flowmaps, parameterName=Velocity Foam Range}
  n10 = Saturate(None=Multiply(A=Divide(A=@n8, B=@n9) {constB=100}, B=ScalarParameter() {defaultValue=0.4, group=Flowmaps, parameterName=Velocity Foam Amount}))
  n11 = ScalarParameter() {defaultValue=-2.0226, parameterName=HeightBias}
  n12 = OneMinus(None=ComponentMask(None=@n4) {g=true})
  Add(
   A=@n7
   B=MaterialFunctionCall(
    Transition Phase=Multiply(A=@n5, B=@n10)
    Height Texture=Multiply(A=@n11, B=@n12)
    Contrast=ScalarParameter() {defaultValue=4, parameterName=Contrast}
   ) {materialFunction=&HeightLerp}.Alpha
  )
  n13 = ScalarParameter() {defaultValue=1024, parameterName=CellSize}
  n14 = Divide(A=@n1.XY, B=@n13)
  n15 = Subtract(A=Frac(None=@n14)) {constB=0.5}
  n16 = Divide(
   A=Subtract(A=Length(None=@n15) {is=class MaterialExpressionLength}, B=ScalarParameter() {defaultValue=0.3386, parameterName=MaskOffset})
   B=ScalarParameter() {defaultValue=0.248, parameterName=MaskBlend}
  )
  n17 = ScalarParameter() {defaultValue=0.5, parameterName=TileOffset}
  n18 = ScalarParameter() {defaultValue=3.2701, parameterName=TileScale}
  n19 = Normalize(VectorInput=NamedRerouteUsage())
  n20 = AppendVector(A=@n19, B=Constant())
  n21 = Constant3Vector() {constant=<rgba 0 0 1 0>}
  n22 = DotProduct(A=@n15, B=ComponentMask(None=CrossProduct(A=@n20, B=@n21)) {g=true, r=true})
  n23 = DotProduct(A=@n15, B=@n19)
  n24 = Add(A=Multiply(A=Floor(None=@n14), B=Multiply(A=@n17, B=@n18)), B=Divide(A=AppendVector(A=@n22, B=@n23), B=@n18))
  n25 = Floor(None=@n14)
  n26 = MaterialFunctionCall!(Texture, Position=Multiply(A=Add(A=@n25) {constB=0.5}, B=@n13)) {materialFunction=/FortWater/Materials/Functions/WaterBodyData}
  n27 = Divide(A=Round(None=ComponentMask(None=@n26.VelocityAndLength) {b=true}), B=@n18)
  n28 = ScalarParameter() {defaultValue=-0.0005, parameterName=PanSpeed}
  n29 = Divide(A=@n26.Water Depth, B=ScalarParameter() {defaultValue=10, parameterName=FlowMapDepthSpeedRange}) {constB=100}
  n30 = Multiply(
   A=AppendVector(A=ScalarParameter() {defaultValue=0.002, parameterName=Pan_X_NonFlowMap}, B=Multiply(A=@n27, B=@n28))
   B=LinearInterpolate(A=ScalarParameter() {defaultValue=0.25, parameterName=FlowMapDepthMinSpeed}, Alpha=Saturate(None=@n29))
  )
  n31 = Frac(None=Multiply(A=Time(), B=@n30))
  n32 = Multiply(
   A=OneMinus(None=Saturate(None=@n16))
   B=TextureSample(UVs=Add(A=@n24, B=@n31)) {texture=/WildEstate/Water/Textures/WildEstate_Water_River_3_M}.R
  )
  n33 = MaterialFunctionCall() {materialFunction=&WaterBodyData}
  n34 = Divide(
   A=Subtract(A=ComponentMask(None=@n33.VelocityAndLength) {b=true}, B=ScalarParameter() {defaultValue=147.6644, parameterName=MaskOffsetVelocity})
   B=ScalarParameter() {defaultValue=154.1066, parameterName=MaskBlendVelocity}
  )
  n35 = ScalarParameter() {defaultValue=1268.288, parameterName=CellSize}
  n36 = Divide(A=@n1.XY, B=@n35)
  n37 = Subtract(A=@n36) {constB=0.5}
  n38 = Subtract(A=Frac(None=@n37)) {constB=0.5}
  n39 = Divide(
   A=Subtract(A=Length(None=@n38), B=ScalarParameter() {defaultValue=0.2549, parameterName=MaskOffset})
   B=ScalarParameter() {defaultValue=0.4, parameterName=MaskBlend}
  )
  n40 = ScalarParameter() {defaultValue=0.5, parameterName=TileOffset}
  n41 = ScalarParameter() {defaultValue=1.336, parameterName=TileScale}
  n42 = Normalize(VectorInput=NamedRerouteUsage())
  n43 = AppendVector(A=@n42, B=Constant())
  n44 = Constant3Vector() {constant=<rgba 0 0 1 0>}
  n45 = DotProduct(A=@n38, B=ComponentMask(None=CrossProduct(A=@n43, B=@n44)) {g=true, r=true})
  n46 = DotProduct(A=@n38, B=@n42)
  n47 = Add(A=Multiply(A=Floor(None=@n37), B=Multiply(A=@n40, B=@n41)), B=Divide(A=AppendVector(A=@n45, B=@n46), B=@n41))
  n48 = Floor(None=@n37)
  n49 = MaterialFunctionCall(Position=Multiply(A=Add(A=@n48) {constB=0.5}, B=@n35)) {materialFunction=&WaterBodyData}
  n50 = Divide(A=Round(None=ComponentMask(None=@n49.VelocityAndLength) {b=true}), B=@n41)
  n51 = ScalarParameter() {defaultValue=-0.0005, parameterName=PanSpeed}
  n52 = Divide(A=@n49.Water Depth, B=ScalarParameter() {defaultValue=10, parameterName=FlowMapDepthSpeedRange}) {constB=100}
  n53 = Multiply(
   A=AppendVector(A=ScalarParameter() {defaultValue=0.002, parameterName=Pan_X_NonFlowMap}, B=Multiply(A=@n50, B=@n51))
   B=LinearInterpolate(A=ScalarParameter() {defaultValue=0.25, parameterName=FlowMapDepthMinSpeed}, Alpha=Saturate(None=@n52))
  )
  n54 = Frac(None=Multiply(A=Time(), B=@n53))
  n55 = Multiply(
   A=OneMinus(None=Saturate(None=@n39))
   B=TextureSample(UVs=Add(A=@n47, B=@n54)) {texture=&WildEstate_Water_River_3_M}.G
  )
  n56 = Multiply(A=NamedRerouteUsage(), B=LinearInterpolate(A=@n32, B=@n55, Alpha=Saturate(None=@n34)))
  n57 = Subtract(A=@n56, B=ScalarParameter() {defaultValue=0.0509, parameterName=FlowMap_ThresholdA})
  n58 = ScalarParameter() {defaultValue=13.5884, parameterName=FlowMap_ThresholdA_Steps}
  n59 = Multiply(A=Saturate(None=@n57), B=@n58)
  n60 = ScalarParameter() {defaultValue=0.1562, parameterName=FlowMap_Smooth}
  n61 = Divide(
   A=MaterialFunctionCall!(In=@n59, Ramp Width=@n60) {materialFunction=/Engine/Functions/Engine_MaterialFunctions01/Math/SmoothCeil}.Result
   B=@n58
  )
  n62 = Saturate(None=Divide(A=Frac(None=@n59), B=@n60))
  Add(A=@n61, B=@n62)
  n63 = ComponentMask(None=NamedRerouteUsage()) {b=true}
  n64 = Subtract(
   A=Multiply(A=Divide(A=@n63) {constB=100}, B=ScalarParameter() {defaultValue=0.4, group=Flowmaps, parameterName=Velocity Foam Amount})
   B=Multiply(A=NamedRerouteUsage(), B=ScalarParameter() {defaultValue=0.002, parameterName=Depth Foam Removal})
  )
  n65 = Divide(
   A=ComponentMask(None=WorldPosition() {worldPositionShaderOffset=WPT_ExcludeAllShaderOffsets}.XYZ) {g=true, r=true}
   B=ScalarParameter() {defaultValue=12707.1055, parameterName=Flowmap Breakup Scale}
  )
  n66 = TextureSample(UVs=@n65) {samplerSource=SSM_Wrap_WorldGroupSettings, samplerType=SAMPLERTYPE_Grayscale, texture=&WildEstate_Water_M}
  n67 = ScalarParameter() {parameterName=RiverFoamWorldSpaceRandomEdges}
  n68 = ScalarParameter() {parameterName=RiverFoamWorldSpaceRandomCenter}
  n69 = Saturate(None=Divide(A=NamedRerouteUsage(), B=ScalarParameter() {defaultValue=100, parameterName=Random Foam DepthRange}))
  Add(A=@n64, B=Multiply(A=@n66.G, B=LinearInterpolate(A=@n67, B=@n68, Alpha=@n69)))
  n70 = NamedRerouteDeclaration(None=Subtract(A=SceneDepthWithoutWater!(UVs).R, B=PixelDepth().R)) {name=Depthbehind Surfac, nodeColor=<rgba 0.7812 1 0 1>}
  n71 = ScalarParameter() {defaultValue=64.4684, group=ShoreWaveLines, parameterName=WildEstate_RiverShore_Gradient}
  n72 = NamedRerouteUsage()
  n73 = ScalarParameter() {defaultValue=-1.7447, group=ShoreWaveLines, parameterName=WildEstate_RiverShore_Distort}
  n74 = ComponentMask(None=NamedRerouteUsage()) {b=true}
  n75 = Add(
   A=Panner(
    Coordinate=Divide(A=@n70, B=@n71)
    Speed=ScalarParameter() {defaultValue=0.29, group=ShoreWaveLines, parameterName=WildEstate_RiverWavePanSpeed}
   )
   B=Multiply(A=Normalize(VectorInput=@n72), B=Multiply(A=@n73, B=@n74))
  )
  n76 = ComponentMask(None=@n75) {r=true}
  n77 = ScalarParameter() {defaultValue=0.9323, parameterName=SineLinesThickness}
  n78 = Ceil(None=Subtract(A=Sine(None=@n76), B=@n77))
  n79 = NamedRerouteUsage()
  n80 = LinearInterpolate(
   A=ScalarParameter() {defaultValue=-2.3092, group=ShoreWaveLines, parameterName=WildEstate_LargeMask_low}
   B=ScalarParameter() {defaultValue=4.2249, group=ShoreWaveLines, parameterName=WildEstate_LargeMask_high}
   Alpha=@n79
  )
  n81 = LinearInterpolate(
   A=ScalarParameter() {defaultValue=3.7604, group=ShoreWaveLines, parameterName=WildEstate_LargeMask_B_low}
   B=ScalarParameter() {defaultValue=-6.3644, group=ShoreWaveLines, parameterName=WildEstate_LargeMask_B_high}
   Alpha=@n79
  )
  n82 = Add(A=@n75, B=ScalarParameter() {defaultValue=0.48, group=ShoreWaveLines, parameterName=WildEstate_RiverWavePanOffset})
  n83 = ComponentMask(None=@n82) {r=true}
  n84 = Ceil(None=Subtract(A=Sine(None=@n83), B=@n77))
  n85 = Add(A=Multiply(A=Saturate(None=@n78), B=Saturate(None=@n80)), B=Multiply(A=Saturate(None=@n81), B=Saturate(None=@n84)))
  ComponentMask(None=@n85) {g=true}
  Constant()
  Constant()
  n86 = TextureCoordinate()
  DDX(Value=@n86)
  DDY(Value=@n86)
  Floor(None=@n59)
  n87 = Saturate(None=@n61)
  n88 = Saturate(None=Divide(A=@n57, B=ScalarParameter() {defaultValue=0.005, parameterName=FlowMap_Smooth_blend}))
  FunctionOutput(None=NamedRerouteDeclaration(None=Multiply(A=@n87, B=@n88)) {name=FlowMapTexture, nodeColor=<rgba 0 1 0.6484 1>}) {outputName=FlowMapTexture, sortPriority=999}
  n89 = NamedRerouteUsage()
  n90 = ScalarParameter() {group=Flowmaps, parameterName=Wave Z Displacement}
  FunctionOutput(None=AppendVector(A=Constant2Vector(), B=Multiply(A=@n89, B=@n90))) {outputName=Flowmap WPO, sortPriority=1}
  n91 = FunctionInput() {inputName=Attributes, inputType=FunctionInput_MaterialAttributes}
  n92 = GetMaterialAttributes(None=@n91) {attributeGetTypes [#1, #2]}
  n93 = ComponentMask(None=NamedRerouteUsage()) {b=true}
  n94 = Multiply(A=ScalarParameter() {defaultValue=4.0994, parameterName=ShoreWaveMaskMUlt}, B=@n93)
  n95 = ComponentMask(None=@n6) {g=true}
  n96 = Add(A=Multiply(A=NamedRerouteUsage(), B=Saturate(None=@n94)), B=LinearInterpolate(A=@n7, B=@n95, Alpha=@n10))
  n97 = Multiply(
   A=@n96
   B=Power(Base=FunctionInput() {bUsePreviewValueAsDefault=true, inputName=Transition, inputType=FunctionInput_Scalar}) {constExponent=8}
  )
  n98 = StaticSwitchParameter(True=@n97, False=StaticSwitchParameter(True=@n97, False=@n96) {parameterName=Ocean River Transition}) {parameterName=River Transition}
  n99 = ShadingPathSwitch(Default=@n98, Mobile=Multiply(A=ScalarParameter() {defaultValue=0.25, parameterName=WildEstate_WaveLines_Mobile_Intensity}, B=@n98)) {desc=Migrated from FeatureLevelSwitch}
  n100 = LinearInterpolate(A=@n92.BaseColor, B=VectorParameter() {parameterName=WildEstate_Color_WaveLines, defaultValue=<rgba 0.6087 0.7291 0.9167 1>}.RGB, Alpha=@n99) {constAlpha=0}
  n101 = LinearInterpolate(A=@n92.Opacity, B=ScalarParameter() {defaultValue=1, parameterName=Opacity_WildEstate_WaveLines}, Alpha=@n99) {constAlpha=0}
  FunctionOutput(None=StaticSwitchParameter(
    True=@n91
    False=SetMaterialAttributes!(MaterialAttributes=@n92.MaterialAttributes, BaseColor=@n100, Opacity=@n101) {attributeSetTypes [#1, #2]}
   ) {group=Foam and Pondscum, parameterName=DisableRiverFoam}) {outputName=Attributes}
  n102 = FunctionInput() {bUsePreviewValueAsDefault=true, inputName=Sim Foam Erasure, inputType=FunctionInput_Scalar, sortPriority=1}
  n103 = ScalarParameter() {defaultValue=-1.838, group=Flowmaps, parameterName=Foam_SimErasure_Min}
  n104 = Subtract(A=@n64, B=Multiply(A=Add(A=@n102, B=@n103), B=ScalarParameter() {defaultValue=0.152, group=Flowmaps, parameterName=Foam_SimErasure_Range}))
  n105 = ScalarParameter() {defaultValue=4.5, parameterName=River Foam Phase Mip Scatter}
  n106 = Subtract(
   A=Add(
    A=@n104
    B=TextureSample!(UVs=@n65, Tex, Level=@n105, Apply View MipBias) {mipValueMode=TMVM_MipLevel, samplerSource=SSM_Wrap_WorldGroupSettings, texture=/FortWater/Textures/T_WaterFlow_01_Foam_Tiled}.R
   )
   B=ScalarParameter() {defaultValue=1.25, parameterName=Foam Scattering Offset}
  )
  n107 = VectorParameter() {parameterName=River Foam Scattering, defaultValue=<rgba 0.6 1 1 0.4>}
  n108 = Multiply(A=@n107.RGB, B=Divide(A=@n107.A) {constB=100})
  FunctionOutput(None=Saturate(None=Multiply(A=@n106, B=@n108))) {outputName=Scattering, sortPriority=5}
  n109 = VectorParameter() {parameterName=River Foam Color 1, defaultValue=<rgba 0.7396 0.9349 1 0.45>}
  FunctionOutput(None=NamedRerouteDeclaration(None=Multiply(A=@n109.RGB, B=@n109.RGB)) {name=FinalFoam Color, nodeColor=<rgba 1 0.4922 0 1>}) {outputName=Foam, sortPriority=10}
  n110 = NamedRerouteUsage()
  FunctionOutput(None=AppendVector(A=NamedRerouteUsage(), B=ComponentMask(None=@n110) {r=true})) {
   description="Outputs 4 exponential masks using a single vectorized Exp operation. R=Wave Masked based on Depth, G=Flowmap Mask based on Depth, B=Coloration Mask based on Depth, A=Flow Mask based on velocity"
   outputName=Phase Offset And Foam
   sortPriority=10
  }
  Length(None=@n26.VelocityAndLength)
  Length(None=@n49.VelocityAndLength)
  LinearInterpolate()
  n111 = TextureObjectParameter() {group=Flowmaps, parameterName=FlowmapWaveNormalHeight, texture=&WildEstate_Water_River_3_M}
  n112 = ScalarParameter() {defaultValue=1.5, parameterName=Final Flow Mult}
  n113 = ScalarParameter() {defaultValue=1834.5598, group=Flowmaps, parameterName=FlowmapWaveNormalScale}
  n114 = Divide(A=ComponentMask(None=NamedRerouteUsage()) {g=true, r=true}, B=@n113)
  n115 = ScalarParameter() {defaultValue=0.75, group=Flowmaps, parameterName=Phase Fade Time}
  n116 = LinearInterpolate(A=Divide(A=Multiply(A=@n112, B=@n114), B=@n115), Alpha=NamedRerouteUsage()) {constB=0}
  n117 = VectorParameter() {group=Flowmaps, parameterName=Flowmap UV2 Offset and Scale, defaultValue=<rgba 1 1 1 0>}
  n118 = NamedRerouteUsage()
  MaterialFunctionCall!(
   Diffuse=@n111
   Normal Map
   "Flow Vector Map (see tooltip)"=@n116
   UV texture 2 offset=@n117.RGB
   "Time (see tooltip)"
   useMipLevel
   UVs=@n118
   Texture Size=ScalarParameter() {defaultValue=1024, parameterName=TexSize}
   Texture Mip Bias
  ) {materialFunction=/Engine/Functions/Engine_MaterialFunctions02/Texturing/Flowmaps.FlowMaps}
  n119 = Multiply(
   A=MaterialFunctionCall() {materialFunction=/Water/Materials/Functions/GetWaterTime}.WaterTime
   B=FunctionInput(Preview=ScalarParameter() {defaultValue=0.25, group=FlowMap, parameterName=FLOW_Speed, sortPriority=3}) {bUsePreviewValueAsDefault=true, inputName=Speed, inputType=FunctionInput_Scalar, sortPriority=3}
  )
  n120 = Multiply(A=@n119, B=ScalarParameter() {defaultValue=0.5, parameterName=TimeOffset3rd}) {constB=2}
  n121 = Frac(None=@n120)
  n122 = Multiply(A=Multiply(A=NamedRerouteUsage()) {constB=-1}, B=ScalarParameter() {defaultValue=2, parameterName=FlowmapScale})
  n123 = NamedRerouteUsage()
  n124 = Add(A=Multiply(A=Subtract(A=@n121) {constB=0.5}, B=Divide(A=@n122) {constB=4}), B=@n123)
  n125 = Constant() {r=2}
  n126 = Sine(None=Subtract(A=@n120, B=MaterialFunctionCall(Multiplier=@n125) {materialFunction=&Pi}.Result))
  n127 = OneMinus(None=Divide(A=Add(A=@n126)))
  n128 = ScalarParameter() {defaultValue=0.2, parameterName=3rdFlowOpacity}
  n129 = Frac(None=Add(A=@n119) {constB=0.5})
  n130 = Add(A=Multiply(A=Subtract(A=@n129) {constB=0.5}, B=@n122), B=@n123)
  n131 = Frac(None=@n119)
  n132 = Add(A=Multiply(A=Subtract(A=@n131) {constB=0.5}, B=@n122), B=@n123)
  n133 = Subtract(A=@n119, B=MaterialFunctionCall(Multiplier=Constant() {r=2}) {materialFunction=&Pi}.Result)
  n134 = Divide(A=Add(A=Sine(None=@n133)))
  Max(
   A=LinearInterpolate(A=Constant3Vector(), B=TextureSample!(UVs=@n124, Tex=@n111, Apply View MipBias).RGB, Alpha=Multiply(A=@n127, B=@n128) {constB=0.5})
   B=LinearInterpolate(A=TextureSample(UVs=@n130, Tex=@n111).RGB, B=TextureSample(UVs=@n132, Tex=@n111).RGB, Alpha=Saturate(None=@n134))
  )
  Max(A=@n55, B=@n32)
  n135 = ScalarParameter() {defaultValue=256, group=Flowmaps, parameterName=Target Depth for Max Velocity Flow}
  MaterialExpressionDivide_5 = Divide(
   A=NamedRerouteUsage()
   B=AppendVector(A=Divide(A=@n135), B=ScalarParameter() {defaultValue=128, group=Flowmaps, parameterName=Target Depth for Flowmap Waves})
  )
  n136 = Custom!(x=@MaterialExpressionDivide_5) {
   code="return 1-exp(-x);"
   description=exp
   outputType=CMOT_Float2
   inputs [{inputName=x, input {expression=&self:MaterialExpressionDivide_5}}]
  }
  n137 = ComponentMask(None=@n136) {r=true}
  n138 = ScalarParameter() {defaultValue=0.5, parameterName=MinEdgeDepthFlowMap}
  Min(A=@n137, B=@n138)
  Multiply(A=@n95, B=@n10)
  Multiply(A=@n61, B=@n62)
  n139 = Multiply(A=NamedRerouteUsage(), B=NamedRerouteUsage())
  n140 = Add(
   A=Multiply(
    A=NamedRerouteDeclaration(None=@n139) {name=Height Scaled by Velocity, nodeColor=<rgba 0 1 0.2734 1>}
    B=ScalarParameter() {defaultValue=0.2, parameterName=WPO River Foam}
   )
   B=Multiply(A=NamedRerouteUsage(), B=ScalarParameter() {defaultValue=0.5, parameterName=Foam Texture Amount})
  )
  n141 = Multiply(A=NamedRerouteUsage(), B=ScalarParameter() {defaultValue=1, parameterName=Phase offset Foam Amount})
  n142 = Subtract(A=Add(A=Add(A=@n140, B=@n141), B=@n104), B=ScalarParameter() {defaultValue=1.7, group=Flowmaps, parameterName=River Foam offset})
  n143 = Multiply(A=@n142, B=ScalarParameter() {defaultValue=1, parameterName=River Foam Boost})
  NamedRerouteDeclaration(None=FunctionOutput(None=Saturate(None=@n143)) {outputName=Foam Mask, sortPriority=10}) {name=FoamMask, nodeColor=<rgba 0.125 0 1 1>}
  n144 = Max(A=@n142) {constB=0}
  n145 = ScalarParameter() {defaultValue=16, group=Foam and Pondscum, parameterName=Foam WPO}
  NamedRerouteDeclaration(None=Multiply(A=Multiply(A=@n144, B=@n145), B=Constant3Vector() {constant=<rgba 0 0 1 0>})) {name=Foam WPO, nodeColor=<rgba 0.7812 1 0 1>}
  n146 = MaterialFunctionCall() {materialFunction=&GetWaterTime}
  n147 = MaterialFunctionCall!(
   Diffuse=@n111
   Normal Map
   "Flow Vector Map (see tooltip)"=@n116
   UV texture 2 offset and Scale=@n117.RGB
   "Time (see tooltip)"=Multiply(A=ScalarParameter() {defaultValue=0.272, parameterName=Final Flow Speed}, B=@n146.WaterTime)
   UVs=@n118
  ) {materialFunction=/FortWater/Materials/Functions/WaterFlowMap}
  NamedRerouteDeclaration(None=@n147.Diffuse Alpha) {name=FlowMap Alpha Channel, nodeColor=<rgba 0.8516 0 1 1>}
  n148 = ConstantBiasScale(None=@n147.Diffuse) {bias=-0.5, scale=2}
  n149 = ComponentMask(None=@n148) {b=true}
  NamedRerouteDeclaration(None=@n149) {name=Flow Map Height Channel, nodeColor=<rgba 1 0 0.8672 1>}
  n150 = ComponentMask(None=@n148) {g=true, r=true}
  NamedRerouteDeclaration(None=@n150) {name=Flow Map Normal RG Channel, nodeColor=<rgba 0.9688 1 0 1>}
  NamedRerouteDeclaration(None=@n3.R) {name=WorldSpaceNoise Texture, nodeColor=<rgba 0.1953 1 0 1>}
  n151 = FunctionInput() {bUsePreviewValueAsDefault=true, inputName=Velocity, sortPriority=2, previewValue=<1 0 0 1>}
  n152 = Multiply(A=NamedRerouteDeclaration(None=@n151) {name=InputVelocity, nodeColor=<rgba 0.8281 1 0 1>}, B=NamedRerouteUsage())
  NamedRerouteDeclaration(None=@n152) {name=VelocityScaled, nodeColor=<rgba 0 1 0.4609 1>}
  NamedRerouteDeclaration(None=@n116) {name=FlowMapVector, nodeColor=<rgba 1 0.9844 0 1>}
  NamedRerouteDeclaration(None=@n117.RGB) {name=FlowMapSacle, nodeColor=<rgba 1 0.9844 0 1>}
  n153 = Saturate(None=Divide(A=@n70, B=ScalarParameter() {defaultValue=86.504, group=ShoreWaveLines, parameterName=WildEstate_RiverShore_Blend}))
  n154 = ComponentMask(None=NamedRerouteUsage()) {b=true}
  n155 = Saturate(None=LinearInterpolate(
    A=ScalarParameter() {defaultValue=1, parameterName=River_Edge_Flow_Mask_A}
    B=ScalarParameter() {defaultValue=-4.374, parameterName=River_Edge_Flow_Mask_B}
    Alpha=@n154
   ))
  n156 = ComponentMask(None=@n85) {r=true}
  NamedRerouteDeclaration(None=Multiply(A=OneMinus(None=@n153), B=Multiply(A=@n155, B=@n156))) {name=ShoreWave, nodeColor=<rgba 0.8281 1 0 1>}
  n157 = Subtract(A=Min(A=NamedRerouteUsage(), B=NamedRerouteUsage()), B=ScalarParameter() {defaultValue=50, parameterName=DepthmaskingOffset})
  n158 = ScalarParameter() {defaultValue=215.1797, parameterName=DepthmaskingBlend}
  NamedRerouteDeclaration(None=Saturate(None=Divide(A=@n157, B=@n158))) {name=DepthMaskOffset, nodeColor=<rgba 0.4062 1 0 1>}
  NamedRerouteDeclaration(None=@n56) {name=test, nodeColor=<rgba 0.9922 1 0 1>}
  n159 = NamedRerouteUsage()
  n160 = Divide(A=Max(A=ComponentMask(None=@n159) {b=true}) {constB=0.001}, B=ScalarParameter() {defaultValue=512, parameterName=Velocity for Max Normals})
  n161 = Multiply(
   A=Multiply(A=NamedRerouteUsage(), B=NamedRerouteUsage())
   B=Multiply(A=ScalarParameter() {defaultValue=0.3, group=Flowmaps, parameterName=WavelNormalIntensity}, B=Saturate(None=@n160))
  )
  NamedRerouteDeclaration(None=FunctionOutput(None=DeriveNormalZ(InXY=@n161)) {outputName=Flowmap Normals}) {name=FlowMapNormals, nodeColor=<rgba 0.9453 0 1 1>}
  n162 = FunctionInput(Preview=Constant()) {bUsePreviewValueAsDefault=true, inputName=Far Mask Override, inputType=FunctionInput_Scalar, sortPriority=3}
  NamedRerouteDeclaration(None=@n162) {name=InputFarMaskOverride, nodeColor=<rgba 0 0.5312 1 1>}
  n163 = Divide(
   A=ComponentMask(None=WorldPosition() {worldPositionShaderOffset=WPT_ExcludeAllShaderOffsets}.XYZ) {g=true, r=true}
   B=LinearInterpolate(A=@n113, B=ScalarParameter() {defaultValue=8000, group=Distance Blending, parameterName=FarScale}, Alpha=NamedRerouteUsage())
  )
  NamedRerouteDeclaration(None=@n163) {name=Flow Map UVs, nodeColor=<rgba 0 1 0.0859 1>}
  n164 = FunctionInput() {bUsePreviewValueAsDefault=true, inputName=Depth, inputType=FunctionInput_Scalar, sortPriority=4, previewValue=<1 0 0 1>}
  n165 = Max(A=@n164) {constB=0}
  NamedRerouteDeclaration(None=@n165) {name=Input Depth, nodeColor=<rgba 0.4531 0 1 1>}
  NamedRerouteDeclaration(None=Max(A=@n137, B=@n138)) {name=TargetDepthForMaxVelocity, nodeColor=<rgba 1 0 0.7734 1>}
  n166 = ComponentMask(None=@n136) {g=true}
  NamedRerouteDeclaration(None=@n166) {name=TargetDepth For Flowmap Waves, nodeColor=<rgba 0.2891 1 0 1>}
  NamedRerouteUsage()
  NamedRerouteUsage()
  NamedRerouteUsage()
  NamedRerouteUsage()
  NamedRerouteUsage()
  NamedRerouteUsage()
  NamedRerouteUsage()
  NamedRerouteUsage()
  OneMinus(None=@n62)
  n167 = Multiply(A=OneMinus(None=@n121), B=@n121)
  n168 = ScalarParameter() {group=FlowMap, parameterName=FLOW_BiasBlendTime_3, sortPriority=4}
  n169 = Divide(A=Subtract(A=Multiply(A=@n167) {constB=4}, B=@n168), B=Subtract(B=@n168))
  n170 = Constant() {r=1}
  n171 = Saturate(None=Multiply(A=Saturate(None=@n169), B=Add(A=@n170, B=@n168)))
  n172 = ScalarParameter() {defaultValue=1, group=FlowMap, parameterName=FLOW_BlendFalloff_3rd, sortPriority=5}
  OneMinus(None=Saturate(None=Power(Base=@n171, Exp=@n172)))
  OneMinus(None=@n154)
  OneMinus(None=@n93)
  n173 = Subtract(A=Length(None=@n15), B=ScalarParameter() {parameterName=MaskOffset2})
  n174 = ScalarParameter() {defaultValue=3.2056, parameterName=MaskBlend2}
  OneMinus(None=Saturate(None=Divide(A=@n173, B=@n174)))
  Panner(Coordinate=@n24, Speed=@n30) {speedY=-0.05}
  Panner(Coordinate=@n47, Speed=@n53) {speedY=-0.05}
  Reroute(None=@n151)
  Reroute(None=@n163)
  Reroute(None=@n165)
  Reroute(None=@n166)
  Reroute(None=@n147.Diffuse Alpha)
  Reroute(None=@n149)
  Reroute(None=@n150)
  Reroute(None=@n139)
  Reroute(None=@n152)
  Reroute(None=@n162)
  Round(None=@n36)
  Round(None=@n59)
  n175 = OneMinus(None=@n131)
  n176 = FunctionInput(Preview=ScalarParameter() {group=FlowMap, parameterName=FLOW_BiasBlendTime, sortPriority=4}) {bUsePreviewValueAsDefault=true, inputName=BlendBias_0-1, inputType=FunctionInput_Scalar, sortPriority=4}
  n177 = Subtract(A=Multiply(A=Multiply(A=@n175, B=@n131)) {constB=4}, B=@n176)
  n178 = Subtract(B=@n176)
  n179 = Multiply(A=Saturate(None=Divide(A=@n177, B=@n178)), B=Add(A=Constant() {r=1}, B=@n176))
  n180 = ScalarParameter() {defaultValue=5, group=FlowMap, parameterName=FLOW_BlendFalloff, sortPriority=5}
  Saturate(None=Power(
    Base=Saturate(None=@n179)
    Exp=FunctionInput(Preview=@n180) {bUsePreviewValueAsDefault=true, inputName=BlendExponent, inputType=FunctionInput_Scalar, sortPriority=5}
   ))
  n181 = MaterialFunctionCall(Time=@n146.WaterTime, Speed=@n115) {materialFunction=&TimeWithSpeedVariable}
  n182 = Add(
   A=@n181.Time
   B=Multiply(A=NamedRerouteUsage(), B=ScalarParameter() {defaultValue=1.045, group=Flowmaps, parameterName=Water Flowmap offset Amount})
  ) {constB=0}
  n183 = QualitySwitch(Default=@n182, Low=@n181.Time, High=@n182, Medium=@n181.Time)
  ShadingPathSwitch(Default=@n183, Deferred=@n183, Forward=@n181.Time, Mobile=@n181.Time)
  SmoothStep()
  TextureSample(UVs=@n118, Tex=@n111) {samplerType=SAMPLERTYPE_Masks, texture=/WildEstate/Water/Textures/WildEstate_Water_River_M}
  TextureSample(Tex=@n111) {samplerType=SAMPLERTYPE_Masks, texture=&WildEstate_Water_River_M}
  TextureSample(UVs=@n75) {texture=/WildEstate/Water/Textures/T_WildEstate_WaterWave_M}
  TextureSample(UVs=@n82) {texture=&T_WildEstate_WaterWave_M}
  TextureSample(UVs=@n75) {samplerType=SAMPLERTYPE_Masks, texture=&WildEstate_Water_River_M}
  Time()
  Truncate!(None=@n37) {is=class MaterialExpressionTruncate}
 ]}
